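# Parent material for Underground Rebel gadget props, built as a Default Lit material-attributes graph for skeletal, cloth and mesh-deformer use.
# The shared 2023 character PreFX function produces the base attributes, which pass through the weapon-skin PlayerBanner function into the character PostFX function.
# PreFX also feeds PostFX's auxiliary inputs directly, bypassing the banner step.
# All shading logic lives in these three material functions.
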Material /BRCosmetics/Animation/Props/Gadgets_UndergroundRebel/Materials/M_Gadget_UndergroundRebel_Parent
graph {
 properties {
  bAutomaticallySetUsageInEditor=false
  bUseMaterialAttributes=true
  bUsedWithClothing=true
  bUsedWithMeshDeformer=true
  bUsedWithSkeletalMesh=true
  pixelDepthOffsetMode=PDOM_Legacy
  usedShadingModels=Default Lit
 }
 nodes [
  n1 = MaterialFunctionCall() {materialFunction=&MF_CharacterParent_2023_PreFX}
  MaterialFunctionCall(
   Material=MaterialFunctionCall!(Gmap, Materials=@n1.Result) {materialFunction=/Game/Weapons/WeaponSkins/MaterialFunction/PlayerBanner}.Result
   M=@n1.M
   Pre Skinned Local Position=@n1.M
   Distance Blend=@n1.M
   Fresnel=@n1.M
  ) {materialFunction=&MF_CharacterParent_2023_PostFX}
 ]
}
